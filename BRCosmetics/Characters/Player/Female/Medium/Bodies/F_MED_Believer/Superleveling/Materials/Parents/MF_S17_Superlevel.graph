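# Material function for Season 17 "Superlevel" character cosmetic styles, blending a metal-and-glass finish over input material attributes by Alpha.
# Metal and glass layers each build anisotropic shading with detail normals, combined using SuperlevelMask channels for material and glow.
# A tech pattern, triplanar-projected in pre-skinned space, varies color, roughness and glow.
# Curve-atlas gradients add view-dependent chroma and height-based color, while animated sine and sweep masks drive emissive with time-of-day scaling.
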MaterialFunction /BRCosmetics/Characters/Player/Female/Medium/Bodies/F_MED_Believer/Superleveling/Materials/Parents/MF_S17_Superlevel
graph {nodes [
  n1 = SetMaterialAttributes!(
   MaterialAttributes
   BaseColor=NamedRerouteUsage()
   Metallic=NamedRerouteUsage()
   Specular=ScalarParameter() {defaultValue=1, group=3_Metal, parameterName=Metal_Specular, sortPriority=99}
   Roughness=NamedRerouteUsage()
   Anisotropy=ScalarParameter() {defaultValue=0.1, group=3_Metal, parameterName=Metal_Anisotropy, sortPriority=100}
   Tangent=NamedRerouteUsage()
   EmissiveColor=NamedRerouteUsage()
  ) {attributeSetTypes [
    #1
    #2
    #3
    #4
    #5
    #6
    #7
   ]}
  n2 = SetMaterialAttributes(
   BaseColor=NamedRerouteUsage()
   Metallic=NamedRerouteUsage()
   Specular=ScalarParameter() {defaultValue=1, group=4_Glass, parameterName=Glass_Specular, sortPriority=99}
   Roughness=NamedRerouteUsage()
   Anisotropy=ScalarParameter() {group=4_Glass, parameterName=Glass_Anisotropy, sortPriority=100}
   Tangent=NamedRerouteUsage()
   EmissiveColor=NamedRerouteUsage()
  ) {attributeSetTypes [
    #1
    #2
    #3
    #4
    #5
    #6
    #7
   ]}
  n3 = BlendMaterialAttributes(A=@n1, B=@n2, Alpha=NamedRerouteUsage())
  n4 = NamedRerouteUsage()
  n5 = FunctionInput() {inputName=Material, inputType=FunctionInput_MaterialAttributes}
  FunctionOutput(None=BlendMaterialAttributes(
    A=@n5
    B=SetMaterialAttributes!(MaterialAttributes=@n3, Normal=@n4) {attributeSetTypes [#8]}
    Alpha=FunctionInput() {inputName=Alpha, inputType=FunctionInput_Scalar}
   ))
  n6 = NamedRerouteUsage()
  n7 = ChannelMaskParameter(None=@n6) {group=2_Chroma, maskChannel=Blue, parameterName=ChromaTechPatternChannel, sortPriority=0, defaultValue=<rgba 0 0 1 0>}
  NamedRerouteDeclaration(None=@n7.R) {name=TechPatternGradient, nodeColor=<rgba 0.2187 1 0 1>}
  n8 = NamedRerouteUsage()
  n9 = NamedRerouteUsage()
  n10 = ScalarParameter() {defaultValue=1, group=2_Chroma, parameterName=ChromaExponent}
  NamedRerouteDeclaration(None=LinearInterpolate(A=@n8, B=Add(A=@n8, B=@n9), Alpha=Fresnel(ExponentIn=@n10) {baseReflectFraction=0})) {name=Metal_BaseColor, nodeColor=<rgba 0.8984 1 0 1>}
  n11 = ConstantBiasScale(None=CollectionParameter() {collection=&FortniteMaterialParameters, parameterName=TimeOfDay_0-1}) {bias=-0.5, scale=2}
  NamedRerouteDeclaration(None=LinearInterpolate(B=ScalarParameter() {defaultValue=1, group=5_Glow, parameterName=NightGlowMultiplier}, Alpha=Abs(None=@n11)) {constA=1}) {name=TOD_GlowMultiplier, nodeColor=<rgba 0.3594 0 1 1>}
  n12 = LinearInterpolate(A=NamedRerouteUsage(), B=NamedRerouteUsage(), Alpha=NamedRerouteUsage())
  NamedRerouteDeclaration(None=MaterialFunctionCall(BaseNormal=NamedRerouteUsage(), AdditionalNormal=NamedRerouteDeclaration(None=@n12) {name=DetailNormal, nodeColor=<rgba 1 0 0.6562 1>}) {materialFunction=&BlendAngleCorrectedNormals}.Result) {name=Normal, nodeColor=<rgba 0.7109 1 0 1>}
  n13 = ChannelMaskParameter(None=@n6) {group=3_Metal, maskChannel=Green, parameterName=Metal_BaseColorTechBlockChannel, defaultValue=<rgba 0 1 0 0>}
  NamedRerouteDeclaration(None=@n13.R) {name=TechPatternColor, nodeColor=<rgba 0.0078 1 0 1>}
  n14 = GetMaterialAttributes(None=@n5) {attributeGetTypes [#8, #9]}
  NamedRerouteDeclaration(None=@n14.Normal) {name=IN_Normal, nodeColor=<rgba 0.8281 1 0 1>}
  NamedRerouteDeclaration(None=LinearInterpolate(B=NamedRerouteUsage(), Alpha=ScalarParameter() {defaultValue=0.75, group=1_Base Parameters, parameterName=AO_Strength}) {constA=1}) {name=AmbientOcclusion, nodeColor=<rgba 1 0.7969 0 1>}
  n15 = TextureSampleParameter2D() {group=1_Base Parameters, parameterName=SuperlevelMask, sortPriority=0, texture=&Black}
  NamedRerouteDeclaration(None=@n15.G) {name=MasksG_GlassGlow, nodeColor=<rgba 0 1 0 1>}
  NamedRerouteDeclaration(None=ChannelMaskParameter(None=@n6) {group=3_Metal, parameterName=Metal_RoughnessTechPatternChannel, sortPriority=54}.R) {name=TechPatternRoughness, nodeColor=<rgba 0.9219 1 0 1>}
  n16 = TextureSampleParameter2D(UVs=Multiply(A=TextureCoordinate(), B=ScalarParameter() {defaultValue=8, group=3_Metal, parameterName=Metal_DetailNormalTiling, sortPriority=71})) {
   group=3_Metal
   parameterName=Metal_DetailNormal
   samplerType=SAMPLERTYPE_Normal
   sortPriority=69
   texture=/Game/Global/Textures/Default/Tileable_Patterns/T_BrushedMetal_Detail_N
  }
  n17 = NamedRerouteDeclaration(None=@n16.RGB) {name=Metal_Tangent, nodeColor=<rgba 1 0.0234 0 1>}
  n18 = ScalarParameter() {defaultValue=0.3, group=3_Metal, parameterName=Metal_DetailNormalStrength, sortPriority=70}
  n19 = MaterialFunctionCall(X=@n18, Y=@n18, Z=Constant() {r=1}) {materialFunction=&MakeFloat3}
  NamedRerouteDeclaration(None=Normalize(VectorInput=Multiply(A=@n17, B=@n19.Result))) {name=Metal_DetailNormal, nodeColor=<rgba 1 0 0.5625 1>}
  n20 = TextureCoordinate()
  n21 = ScalarParameter() {defaultValue=16, group=4_Glass, parameterName=Glass_DetailTiling, sortPriority=62}
  n22 = TextureSampleParameter2D(UVs=NamedRerouteDeclaration(None=Multiply(A=@n20, B=@n21)) {name=Glass DetailUVs, nodeColor=<rgba 0.1016 1 0 1>}) {
   group=4_Glass
   parameterName=Glass_DetailNormal
   samplerType=SAMPLERTYPE_Normal
   sortPriority=60
   texture=/Game/Global/Textures/Default/Tileable_Patterns/T_Skyfire_DarkGlass_DetailNormal
  }
  n23 = NamedRerouteDeclaration(None=@n22.RGB) {name=Glass_Tangent, nodeColor=<rgba 0.9688 0 1 1>}
  n24 = ScalarParameter() {defaultValue=0.1, group=4_Glass, parameterName=Glass_DetailNormalStrength, sortPriority=61}
  n25 = MaterialFunctionCall(X=@n24, Y=@n24, Z=Constant() {r=1}) {materialFunction=&MakeFloat3}
  NamedRerouteDeclaration(None=Normalize(VectorInput=Multiply(A=@n23, B=@n25.Result))) {name=Glass_DetailNormal, nodeColor=<rgba 0.1016 0 1 1>}
  n26 = TextureObjectParameter() {
   group=2_TechPattern
   parameterName=TechPatternMap
   samplerType=SAMPLERTYPE_LinearColor
   sortPriority=10
   texture=/BRCosmetics/Characters/Player/Female/Medium/Bodies/F_MED_Believer/Superleveling/Textures/Shared/S17_superlevel_tech_mask_v3
  }
  n27 = ScalarParameter() {defaultValue=64, group=2_TechPattern, parameterName=TechPatternTextureSize, sortPriority=20}
  n28 = MaterialFunctionCall(X=@n27, Y=@n27, Z=@n27) {materialFunction=&MakeFloat3}
  n29 = NamedRerouteDeclaration(None=VertexInterpolator(VS=PreSkinnedPosition()).PS) {name=PreSkinnedPosition, nodeColor=<rgba 0 0.5312 1 1>}
  n30 = StaticBool() {value=true}
  n31 = VertexInterpolator(VS=PreSkinnedNormal())
  n32 = ScalarParameter() {defaultValue=1, group=2_TechPattern, parameterName=TechPatternBlendContrast, sortPriority=21}
  n33 = NamedRerouteUsage()
  NamedRerouteDeclaration(None=Multiply(
    A=MaterialFunctionCall(
     TextureObject=@n26
     TextureSize=@n28.Result
     WorldPosition=@n29
     Export Float 4=@n30
     World Space Normal=@n31.PS
     ProjectionTransitionContrast=@n32
    ) {materialFunction=&WorldAlignedTexture}.XYZ Texture
    B=OneMinus(None=@n33)
   )) {name=TechPatternRGBA, nodeColor=<rgba 0 1 0.1328 1>}
  n34 = Multiply(A=NamedRerouteUsage(), B=ScalarParameter() {defaultValue=0.3, group=2_Chroma, parameterName=ChromaTechPatternGradientScale, sortPriority=49})
  n35 = NamedRerouteUsage()
  n36 = ScalarParameter() {defaultValue=0.2, group=2_Chroma, parameterName=ChromaTechPatternOffset, sortPriority=48}
  n37 = Fresnel(ExponentIn=ScalarParameter() {defaultValue=0.5, group=2_Chroma, parameterName=ChromaWarp, sortPriority=20}) {baseReflectFraction=0}
  n38 = ScalarParameter() {defaultValue=1, group=2_Chroma, parameterName=ChromaScale}
  n39 = ComponentMask(None=MaterialFunctionCall() {materialFunction=&CameraDirectionVector}.Result) {r=true}
  n40 = ScalarParameter() {defaultValue=1, group=2_Chroma, parameterName=ChromaViewDirScale, sortPriority=47}
  n41 = Add(
   A=Add(A=ScalarParameter() {group=2_Chroma, parameterName=ChromaOffset, sortPriority=43}, B=Multiply(A=@n35, B=@n36))
   B=Add(A=Multiply(A=@n37, B=@n38), B=Multiply(A=@n39, B=@n40))
  )
  n42 = CurveAtlasRowParameter(CurveTime=Frac(None=Add(A=@n34, B=@n41))) {
   atlas=/BRCosmetics/Characters/Player/Female/Medium/Bodies/F_MED_Believer/Superleveling/Textures/Curves/C_S17_Superlevel
   curve=/BRCosmetics/Characters/Player/Female/Medium/Bodies/F_MED_Believer/Superleveling/Textures/Curves/C_S17_Superlevel_SilverChroma
   defaultValue=5
   group=2_Chroma
   parameterName=ChromaColorCurve
   sortPriority=0
  }
  NamedRerouteDeclaration(None=Multiply(
    A=NamedRerouteDeclaration(None=@n42) {name=ChromaColor, nodeColor=<rgba 0.1016 0 1 1>}
    B=ScalarParameter() {defaultValue=1, group=2_Chroma, parameterName=ChromaIntensity, sortPriority=40}
   )) {name=ChromaFinal, nodeColor=<rgba 0.7344 0 1 1>}
  n43 = ScalarParameter() {defaultValue=0.3, group=3_Metal, parameterName=Metal_BaseColorTechPatternLow, sortPriority=31}
  n44 = ScalarParameter() {defaultValue=1, group=3_Metal, parameterName=Metal_BaseColorTechPatternHigh}
  n45 = MaterialFunctionCall() {materialFunction=&MF_Fresnel_FX}
  n46 = Add(A=@n45.Fresnel, B=NamedRerouteUsage())
  n47 = Multiply(A=@n45.Fresnel Inverted, B=ConstantBiasScale(None=Sine(None=@n46) {period=0.5}))
  n48 = VectorParameter() {group=3_Metal, parameterName=Metal_BaseColor, sortPriority=0, defaultValue=<rgba 0.75 0.75 0.75 1>}
  n49 = NamedRerouteUsage()
  n50 = ScalarParameter() {defaultValue=1, group=3_Metal, parameterName=Metal_ColorGradientStrength, sortPriority=2}
  NamedRerouteDeclaration(None=Multiply(A=LinearInterpolate(A=@n43, B=@n44, Alpha=@n47), B=LinearInterpolate(A=@n48.RGB, B=@n49, Alpha=@n50))) {name=Metal_Color, nodeColor=<rgba 0 0.0156 1 1>}
  n51 = ComponentMask(None=NamedRerouteUsage()) {b=true}
  n52 = ScalarParameter() {defaultValue=1, group=3_Metal, parameterName=Metal_ColorGradientUVTile, sortPriority=4}
  n53 = PreSkinnedLocalBounds()
  n54 = Add(
   A=Divide(A=Multiply(A=@n51, B=@n52), B=ComponentMask(None=@n53.Extents) {b=true})
   B=ScalarParameter() {group=3_Metal, parameterName=Metal_ColorGradientUVOffset, sortPriority=3}
  )
  n55 = Multiply(
   A=ComponentMask(None=MaterialFunctionCall() {materialFunction=&CameraDirectionVector}.Result) {r=true}
   B=Multiply(A=NamedRerouteUsage(), B=ScalarParameter() {group=3_Metal, parameterName=Metal_ColorGradientUVTechPatternOffset, sortPriority=5})
  )
  n56 = CurveAtlasRowParameter(CurveTime=Saturate(None=Add(A=@n54, B=@n55))) {
   atlas=&C_S17_Superlevel
   curve=/BRCosmetics/Characters/Player/Female/Medium/Bodies/F_MED_Believer/Superleveling/Textures/Curves/C_S17_Superlevel_Silver
   defaultValue=2
   group=3_Metal
   parameterName=Metal_ColorGradientCurve
   sortPriority=0
  }
  NamedRerouteDeclaration(None=@n56) {name=VerticalColorGradient, nodeColor=<rgba 0.125 0 1 1>}
  n57 = NamedRerouteUsage()
  n58 = CurveAtlasRowParameter(CurveTime=NamedRerouteUsage()) {
   atlas=/Game/Environments/Apollo/Sets/Skyfire/Curves/SkyfireCurveAtlas
   curve=/Game/Environments/Apollo/Sets/Skyfire/Curves/C_Skyfire_DarkColor_Diffuse
   defaultValue=3
   group=4_Glass
   parameterName=Glass_BaseColorCurve
   primitiveDataIndex=1
  }
  NamedRerouteDeclaration(None=Multiply(A=Multiply(A=@n57, B=@n58), B=ScalarParameter() {defaultValue=1, group=4_Glass, parameterName=Glass_BaseColorDarken})) {name=Glass_BaseColor, nodeColor=<rgba 0 0.4844 1 1>}
  n59 = TextureSampleParameter2D(UVs=NamedRerouteUsage()) {
   parameterName=PackB
   samplerSource=SSM_Wrap_WorldGroupSettings
   samplerType=SAMPLERTYPE_LinearColor
   texture=/Game/Environments/Apollo/Sets/Skyfire/Textures/T_Skyfire_PackB
  }
  NamedRerouteDeclaration(None=LinearInterpolate(
    A=ScalarParameter() {defaultValue=0.22, group=4_Glass, parameterName=Glass_RoughnessLow, sortPriority=40}
    B=ScalarParameter() {defaultValue=0.25, group=4_Glass, parameterName=Glass_RoughnessHigh, sortPriority=41}
    Alpha=@n59.R
   )) {name=Glass_Roughness, nodeColor=<rgba 0 0.9062 1 1>}
  NamedRerouteDeclaration(None=LinearInterpolate(
    A=ScalarParameter() {defaultValue=0.8, group=4_Glass, parameterName=Glass_MetallicLow, sortPriority=50}
    B=ScalarParameter() {defaultValue=0.7, group=4_Glass, parameterName=Glass_MetallicHigh, sortPriority=51}
    Alpha=@n59.R
   )) {name=Glass_Metallic, nodeColor=<rgba 0.875 1 0 1>}
  NamedRerouteDeclaration(None=LinearInterpolate(
    A=ScalarParameter() {defaultValue=0.25, group=3_Metal, parameterName=Metal_RoughnessLow, sortPriority=55}
    B=ScalarParameter() {defaultValue=0.4, group=3_Metal, parameterName=Metal_RoughnessHigh, sortPriority=56}
    Alpha=NamedRerouteUsage()
   )) {name=Metal_Roughness, nodeColor=<rgba 0.3594 0 1 1>}
  NamedRerouteDeclaration(None=LinearInterpolate(
    A=ScalarParameter() {defaultValue=0.85, group=3_Metal, parameterName=Metal_MetallicLow, sortPriority=60}
    B=ScalarParameter() {defaultValue=0.5, group=3_Metal, parameterName=Metal_MetallicHigh, sortPriority=61}
    Alpha=NamedRerouteUsage()
   )) {name=Metal_Metallic, nodeColor=<rgba 0 0.4375 1 1>}
  NamedRerouteDeclaration(None=@n15.R) {name=MasksR_Material, nodeColor=<rgba 0.8984 0 1 1>}
  n60 = LinearInterpolate(
   A=VectorParameter() {group=5_Glow, parameterName=GlowColor, sortPriority=0, defaultValue=<rgba 0 0 0 1>}.RGB
   B=NamedRerouteUsage()
   Alpha=ScalarParameter() {group=5_Glow, parameterName=GlowUseBaseColor, sortPriority=1}
  )
  n61 = Multiply(
   A=Multiply(A=NamedRerouteDeclaration(None=@n60) {name=GlowColor, nodeColor=<rgba 0 0.0625 1 1>}, B=NamedRerouteUsage())
   B=ScalarParameter() {defaultValue=1, group=5_Glow, parameterName=Glass_GlowStrength}
  )
  n62 = ScalarParameter() {defaultValue=0.5, group=5_Glow, parameterName=Glass_GlowTechPatternMinBrightness}
  n63 = Time()
  n64 = ScalarParameter() {defaultValue=0.2, group=5_Glow, parameterName=Glass_GlowTechPatternSpeed}
  n65 = NamedRerouteUsage()
  n66 = NamedRerouteUsage()
  n67 = Sine(None=Add(A=Multiply(A=@n63, B=@n64), B=Add(A=@n65, B=@n66)))
  n68 = Add(A=@n62, B=Multiply(A=OneMinus(None=@n62), B=ConstantBiasScale(None=@n67)))
  NamedRerouteDeclaration(None=Multiply(A=NamedRerouteUsage(), B=Multiply(A=@n61, B=@n68))) {name=Glass_Emissive, nodeColor=<rgba 0 0.4141 1 1>}
  NamedRerouteDeclaration(None=@n15.B) {name=MasksB_MetalGlow, nodeColor=<rgba 0.4062 1 0 1>}
  n69 = ChannelMaskParameter(None=@n6) {group=5_Glow, maskChannel=Alpha, parameterName=GlowTechPatternChannel, sortPriority=0, defaultValue=<rgba 0 0 0 1>}
  NamedRerouteDeclaration(None=@n69.R) {name=TechPatternGlow, nodeColor=<rgba 0 0.5547 1 1>}
  n70 = OneMinus(None=NamedRerouteUsage())
  n71 = ScalarParameter() {defaultValue=2.5, group=5_Glow, parameterName=Metal_GlowFresnelExponent}
  n72 = SmoothStep(
   Min=ScalarParameter() {defaultValue=0.75, group=5_Glow, parameterName=Metal_GlowTopMaskBrightnessMin}
   Value=ComponentMask(None=NamedRerouteUsage()) {b=true}
  )
  n73 = Multiply(A=NamedRerouteUsage(), B=Fresnel(ExponentIn=LinearInterpolate(A=@n71, Alpha=@n72), BaseReflectFractionIn=Constant()))
  n74 = ScalarParameter() {group=5_Glow, parameterName=Metal_GlowStrength}
  n75 = Add(A=Multiply(A=Time(), B=ScalarParameter() {defaultValue=0.5, group=5_Glow, parameterName=Metal_GlowTechPatternSpeed}), B=NamedRerouteUsage())
  n76 = Multiply(A=NamedRerouteUsage(), B=ScalarParameter() {defaultValue=0.5, group=5_Glow, parameterName=Metal_GlowTechPatternTimeOffset})
  n77 = ConstantBiasScale(None=Sine(None=Add(A=@n75, B=@n76)))
  n78 = Multiply(A=ScalarParameter() {defaultValue=0.5, group=5_Glow, parameterName=Metal_GlowTechPatternMaskStrength}, B=NamedRerouteUsage())
  n79 = Multiply(A=Time(), B=ScalarParameter() {defaultValue=0.2, group=5_Glow, parameterName=Metal_GlowAnimatedMaskSpeed})
  n80 = Divide(A=NamedRerouteUsage(), B=PreSkinnedLocalBounds().Max)
  n81 = Multiply(
   A=ComponentMask(None=NamedRerouteDeclaration(None=@n80) {name=NormalizedBounds, nodeColor=<rgba 0 0.2969 1 1>}) {b=true}
   B=ScalarParameter() {defaultValue=1, group=5_Glow, parameterName=Metal_GlowAnimatedMaskScale}
  )
  n82 = ConstantBiasScale(None=Frac(None=Add(A=@n79, B=@n81))) {bias=-0.5, scale=2}
  n83 = Multiply(
   A=Multiply(A=Multiply(A=@n73, B=@n74), B=Add(A=@n77, B=@n78))
   B=SmoothStep(Min=ScalarParameter() {defaultValue=0.5, group=5_Glow, parameterName=Metal_GlowAnimatedMaskMin}, Value=Abs(None=@n82))
  )
  NamedRerouteDeclaration(None=Multiply(A=NamedRerouteUsage(), B=Multiply(A=@n70, B=@n83))) {name=Metal_Emissive, nodeColor=<rgba 1 0.1172 0 1>}
  NamedRerouteDeclaration(None=@n14.AmbientOcclusion) {name=IN_AO, nodeColor=<rgba 1 0 0.8672 1>}
 ]}
