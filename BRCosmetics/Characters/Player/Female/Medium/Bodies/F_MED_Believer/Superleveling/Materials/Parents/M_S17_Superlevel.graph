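# Parent character material for the Season 17 Superlevel look on the F_MED_Believer body, for skeletal meshes with Default Lit and subsurface-profile skin shading.
# Material attributes flow from the character PreFX function through MF_S17_Superlevel, then PostFX and the CINE character post-effect function.
# A static switch, SuperlevelUseSkinMask, defaults to using the inverted skin mask as the Superlevel alpha, excluding skin; otherwise the alpha is constant full.
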Material /BRCosmetics/Characters/Player/Female/Medium/Bodies/F_MED_Believer/Superleveling/Materials/Parents/M_S17_Superlevel
graph {
 properties {
  bAutomaticallySetUsageInEditor=false
  bUseMaterialAttributes=true
  bUsedWithMeshDeformer=true
  bUsedWithMorphTargets=true
  bUsedWithSkeletalMesh=true
  pixelDepthOffsetMode=PDOM_Legacy
  shadingModel=MSM_FromMaterialExpression
  subsurfaceProfile=&SS_HeroSkin_02
  usedShadingModels=Default Lit | Subsurface Profile
 }
 nodes [
  n1 = MaterialFunctionCall() {materialFunction=&BaseFN_CharacterFunction_PreFX}
  n2 = OneMinus(None=@n1.Skin)
  n3 = Constant() {r=1}
  n4 = MaterialFunctionCall(
   AO=@n1.AO
   Distance Blend=@n1.AO
   Fresnel=@n1.AO
   Gmap=@n1.AO
   Pre Skinned Local Position=@n1.AO
   Material=MaterialFunctionCall!(Alpha=StaticSwitchParameter(True=@n2, False=@n3) {defaultValue=true, parameterName=SuperlevelUseSkinMask}, Material=@n1.Result) {materialFunction=&/MF_S17_Superlevel}.Result
  ) {materialFunction=&BaseFN_CharacterFunction_PostFX}
  MaterialFunctionCall(In=@n4.Result) {materialFunction=&CINE_MF_CharPostFX}
 ]
}
